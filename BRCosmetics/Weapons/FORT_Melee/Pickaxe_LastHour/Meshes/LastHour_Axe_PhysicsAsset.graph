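# Physics asset for the LastHour pickaxe mesh, a two-handed melee weapon cosmetic with a five-joint chain set up for physical simulation.
# The handle body combines a capsule and two boxes, and each chain joint has a single box or capsule body.
# Constraints link the joints in series from the handle, starting with a narrow single-axis swing at the first link and loosening progressively toward the free end.
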PhysicsAsset /BRCosmetics/Weapons/FORT_Melee/Pickaxe_LastHour/Meshes/LastHour_Axe_PhysicsAsset
bodies {
 Melee_TwoHanded_Handle [
  {
   shapeType=Capsule
   length=8.1521
   radius=1.2
   shapeName=Melee_TwoHanded_Handle_capsule
   center=<0 -0.326 -21.2691>
   rotation=<rot -0.0001 0 5>
  }
  {
   shapeType=Box
   extentX=2.5
   extentY=5.5
   extentZ=7.5
   shapeName=Melee_TwoHanded_Handle_box
   center=<-0.2779 0.0077 -15.7856>
   rotation=<rot -0.0026 0 -4.1758>
  }
  {
   shapeType=Box
   extentX=1.5
   extentY=10.5
   extentZ=9.5
   shapeName=Melee_TwoHanded_Handle_box
   center=<0.7679 2.9272 -22.6098>
   rotation=<rot -0.0001 0 -5>
  }
 ]
 dyn_chain_1_joint [{shapeType=Box, extentX=2.5, extentY=3.5, extentZ=0.5, shapeName=dyn_chain_1_joint_box, center=<4.2873 0.0837 0.0896>}]
 dyn_chain_2_joint [{
   shapeType=Capsule
   length=0.5
   radius=0.5
   shapeName=dyn_chain_2_joint_capsule
   center=<0.8987 0.0588 -0.1569>
   rotation=<rot 90 179.789 179.7887>
  }]
 dyn_chain_3_joint [{shapeType=Capsule, length=0.5, radius=0.5, shapeName=dyn_chain_3_joint_capsule, center=<0.9552 0 0.028>, rotation=<rot 90 180 179.9997>}]
 dyn_chain_4_joint [{shapeType=Capsule, length=0.505, radius=0.505, shapeName=dyn_chain_4_joint_capsule, center=<0.8772 0.1064 0>, rotation=<rot 0 -89.6548 -90>}]
 dyn_chain_5_joint [{shapeType=Box, extentX=3.5, extentY=0.5, extentZ=2.5, shapeName=dyn_chain_5_joint_box, center=<1.3645 0.0501 0>}]
}
constraints [
 {bone1Name=dyn_chain_1_joint, bone2Name=Melee_TwoHanded_Handle, swing1LimitDegrees=15, swing2LimitDegrees=0, twistLimitDegrees=0}
 {bone1Name=dyn_chain_2_joint, bone2Name=dyn_chain_1_joint, swing1LimitDegrees=60, swing2LimitDegrees=60, twistLimitDegrees=20}
 {bone1Name=dyn_chain_3_joint, bone2Name=dyn_chain_2_joint, swing1LimitDegrees=80, swing2LimitDegrees=80, twistLimitDegrees=30}
 {bone1Name=dyn_chain_4_joint, bone2Name=dyn_chain_3_joint, swing1LimitDegrees=80, swing2LimitDegrees=80, twistLimitDegrees=80}
 {bone1Name=dyn_chain_5_joint, bone2Name=dyn_chain_4_joint, swing1LimitDegrees=90, swing2LimitDegrees=90, twistLimitDegrees=90}
]
